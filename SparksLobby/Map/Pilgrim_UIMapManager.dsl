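; Pilgrim_UIMapManager is a map manager Blueprint in SparksLobby that controls in-game map display.
; Its construction script turns off both AllowShowMainMap and AllowShowMiniMap on the InGameMap, so neither the main map nor the minimap is allowed to show.
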(fn ConstructionScript
  (InGameMap|AllowShowMainMap false)
  (InGameMap|AllowShowMiniMap false))
